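# Animation Blueprint for the FunkyFeather Beak pickaxe that makes its two duck bones float and sway.
# Each update builds differently phased sine offsets from game time, added as translations to each bone before spring-based Anim Dynamics adds secondary motion.
# The dynamics alpha eases lower while jumping or falling and higher otherwise.
# In frontend display, the event graph also sets the IsInFE parameter on the dynamic material.
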AnimBlueprint /BRCosmetics/Weapons/FORT_Melee/Pickaxe_FunkyFeather_Beak/Meshes/M_FunkyFeather_Beak_Axe_AnimBP
target_skeleton=&/M_FunkyFeather_Beak_Axe_Skeleton
graphs [
 AnimGraph {nodes [
   AnimGraphNode_SubInput_0 = Input Pose()
   n1 = "Transform (Modify) Bone - Bone: dyn_duck_01"!(
    ComponentPose=Local To Component(LocalPose=@AnimGraphNode_SubInput_0)
    bAlphaBoolEnabled:bool=True
    Translation=Get Duck_1 Floating() -> (Duck_1 Floating:Vector)
    Rotation="0,0,0"
    Scale="1,1,1"
   ) -> (Pose) {
    alphaInputType=Bool
    rotationMode=BMM_Replace
    rotationSpace=BCS_BoneSpace
    translationMode=BMM_Additive
    translationSpace=BCS_BoneSpace
    alphaBoolBlend {blendInTime=0.2, blendOption=Linear, blendOutTime=0.2}
    boneToModify {boneName=dyn_duck_01}
   }
   n2 = "Transform (Modify) Bone - Bone: dyn_duck_1_01"!(
    ComponentPose=@n1
    bAlphaBoolEnabled:bool=True
    Translation=Get Duck_2 Floating() -> (Duck_2 Floating:Vector)
    Rotation="0,0,0"
    Scale="1,1,1"
   ) -> (Pose) {
    alphaInputType=Bool
    rotationMode=BMM_Replace
    rotationSpace=BCS_BoneSpace
    translationMode=BMM_Additive
    translationSpace=BCS_BoneSpace
    alphaBoolBlend {blendInTime=0.2, blendOption=Linear, blendOutTime=0.2}
    boneToModify {boneName=dyn_duck_1_01}
   }
   n3 = Anim Dynamics - Bone: dyn_duck_1_01(ComponentPose=@n2, Alpha=Get AlphaJump() -> (AlphaJump:double), ExternalForce="0,0,0") {
    angularBiasOverride=0.99
    angularDampingOverride=0.99
    angularSpringConstant=10
    bLinearSpring=true
    bOverrideLinearDamping=true
    bUseGravityOverride=true
    gravityScale=0.15
    linearDampingOverride=0.95
    linearSpringConstant=10
    alphaBoolBlend {blendOption=Linear}
    alphaScaleBiasClamp {bInterpResult=true, interpSpeedDecreasing=5, interpSpeedIncreasing=5}
    boundBone {boneName=dyn_duck_1_01}
    componentAppliedLinearAccClamp=<40 40 40>
    componentLinearAccScale=<0.2 0.2 0.1>
    componentLinearVelScale=<0.2 0.2 0.1>
    physicsBodyDefinitions [{
      boundBone {boneName=dyn_duck_1_01}
      boxExtents=<3 3 3>
      constraintSetup {angularLimitsMax=<0 0 0>, angularLimitsMin=<0 0 0>, angularTarget=<1 0 0>, linearAxesMax=<1 1 1>, linearAxesMin=<-1 -4 -1>}
      localJointOffset=<0 0 0>
     }]
   }
   n4 = Anim Dynamics - Bone: dyn_duck_01(ComponentPose=@n3, Alpha=Get AlphaJump(), ExternalForce="0,0,0") {
    angularBiasOverride=0.99
    angularDampingOverride=0.99
    angularSpringConstant=10
    bLinearSpring=true
    bOverrideLinearDamping=true
    bUseGravityOverride=true
    gravityScale=0.15
    linearDampingOverride=0.95
    linearSpringConstant=7
    alphaBoolBlend {blendOption=Linear}
    alphaScaleBiasClamp {bInterpResult=true, interpSpeedDecreasing=5, interpSpeedIncreasing=5}
    boundBone {boneName=dyn_duck_01}
    componentAppliedLinearAccClamp=<40 40 40>
    componentLinearAccScale=<0.2 0.2 0.1>
    componentLinearVelScale=<0.2 0.2 0.1>
    physicsBodyDefinitions [{
      boundBone {boneName=dyn_duck_01}
      boxExtents=<4 4 6>
      constraintSetup {angularLimitsMax=<0 0 0>, angularLimitsMin=<0 0 0>, angularTarget=<1 0 0>, linearAxesMax=<1 4 1>, linearAxesMin=<-1 -1 -1>}
      localJointOffset=<0 0 0>
     }]
   }
   Output Pose(Result=Component To Local(ComponentPose=@n4)) {layerGroup=None}
  ]}
 EventGraph {nodes [
   n1 = Event BlueprintUpdateAnimation()
   > Set Delta Time X(Delta Time X=@n1.DeltaTimeX)
   > n2 = Sequence!(execute) -> (then_0, then_1, then_2:Exec)
   n3 = float * float(A=GetGameTimeInSeconds(), B=3) {is=title K2Node_PromotableOperator}
   n4 = "Sin (Radians)"(A=float * float(A=0.5, B=@n3) {is=title K2Node_CommutativeAssociativeBinaryOperator})
   n5 = "Sin (Radians)"(A=float * float(A=1, B=@n3))
   n6 = "Sin (Radians)"(A=float * float(A=0.4, B=@n3))
   > Set Duck_1 Floating(Duck_1 Floating:Vector=MakeVector(X=float * float(A=@n6, B=0.7), Y=float * float(A=@n5, B=0.75), Z=float * float(A=@n4, B=1))) -> (Output_Get:Vector)
   n7 = float * float(A=GetGameTimeInSeconds(), B=3) {is=title K2Node_PromotableOperator}
   n8 = "Sin (Radians)"(A=float * float(A=0.7, B=@n7) {is=title K2Node_CommutativeAssociativeBinaryOperator})
   n9 = "Sin (Radians)"(A=float * float(A=0.4, B=@n7))
   n10 = "Sin (Radians)"(A=float * float(A=1.37, B=@n7))
   > Set Duck_2 Floating(Duck_2 Floating:Vector=MakeVector(X=float * float(A=@n8, B=0.7), Y=float * float(A=@n10, B=0.55), Z=float * float(A=@n9, B=1))) -> (Output_Get:Vector)
   n11 = GetPartSkeletalMeshComponent()
   Branch(execute=@n2.then_2, Condition=IsCosmeticObjectDisplayedInFrontendContext!(CosmeticObject:Object Reference=@n11) -> (ReturnValue:bool))
   > n12 = Get Dynamic Material Instance(SkeletalMeshComponent=@n11)
   > SetScalarParameterValue(self=@n12, ParameterName=IsInFE, Value=1)
   n13 = OR Boolean(A=Get bIsJumping(), B=Get bIsFalling())
   Set AlphaJump(
    execute=@n2.then_1
    AlphaJump:double=FInterpTo(Current=Get AlphaJump(), Target=SelectFloat(A=0.2, B=0.7, bPickA=@n13), DeltaTime=Get Delta Time X(), InterpSpeed=5)
   ) -> (Output_Get:double)
   TryGetPawnOwner()
  ]}
]
variables [
 AlphaJump {type=number}
 Delta Time X {type=number}
 Duck_1 Floating {type=Vector}
 Duck_2 Floating {type=Vector}
]
